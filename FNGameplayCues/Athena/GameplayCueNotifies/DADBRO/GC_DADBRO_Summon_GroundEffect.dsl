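; Gameplay cue for the DADBRO boss summon, sending crawling bolt effects with audio from the boss hand to spawner locations.
; The spawner location comes from the SourceObject actor, or from Normal times RawMagnitude, and a downward trace grounds the hand point.
; A shared timeline lerps all stored bolts, then spawns active effects, fades audio and clears arrays except hand locations.
; EndPlay also fades audio; bDebug enables debug drawing.
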(event GameplayCueNotify|EventHandleGameplayCue
  (params (MyTarget :type Actor%20Object%20Reference) (EventType :type EGameplayCueEvent%20Enum :default "OnActive") (Parameters :type Gameplay%20Cue%20Parameters%20Structure%20%28by%20ref%29))
  (bind (_normalizedmagnitude _rawmagnitude _effectcontext _matchedtagname _originaltag _aggregatedsourcetags _aggregatedtargettags _location _normal _instigator _effectcauser _sourceobject) (Ability|GameplayCue|BreakGameplayCueParameters Parameters))
  (bind _crawlingbolthandlocations (Variables|Default|GetCrawlingBoltHandLocations))
  (bind _crawlingboltspawnerlocations (Variables|Default|GetCrawlingBoltSpawnerLocations))
  (Utilities|FlowControl|Sequence
    (:then_0
      (bind _asactor (Utilities|Casting|CastToActor _sourceobject)
        (:then
          (Variables|Default|SetSpawnerLocation (Transformation|GetActorLocation _asactor)))
        (:CastFailed
          (Variables|Default|SetSpawnerLocation (* _normal _rawmagnitude)) ; DADBRO passes the spawner location through as the normal and raw magnitude
        ))
      (bind _outhit (Collision|LineTraceByChannel :Start (Math|Vector|MakeVector (.x _location) (.y _location) (.z _location)) :End (+ (Math|Vector|MakeVector (.x _location) (.y _location) (.z _location)) "0, 0,-3000.000000") :TraceChannel "TraceTypeQuery7" :bTraceComplex false :DrawDebugType (select (Variables|Default|GetbDebug) "Persistent" "None") :bIgnoreSelf true :TraceColor "(R=1.000000,G=0.000000,B=0.000000,A=1.000000)" :TraceHitColor "(R=0.000000,G=1.000000,B=0.000000,A=1.000000)" :DrawTime 5.0))
      (bind (_outhit_bblockinghit _outhit_binitialoverlap _outhit_time _outhit_distance _outhit_location) (Collision|BreakHitResult _outhit))
      (Variables|Default|SetBossHandLocation (+ _outhit_location "0, 0,131.000000")))
    (:then_1
      (bind _returnvalue (Effects|SpawnEmitteratLocation 0 "/FNGameplayCues/Athena/DADBRO/FX/Summon/P_DadBRO_Summon_CrawlingBolt.P_DadBRO_Summon_CrawlingBolt" (Variables|Default|GetBossHandLocation) (Math|Rotator|FindLookatRotation (Variables|Default|GetBossHandLocation) (Variables|Default|GetSpawnerLocation))))
      (Utilities|Array|Add (Variables|Default|GetCrawlingBoltArray) _returnvalue)
      (Utilities|Array|Add (Variables|Default|GetCrawlingBoltSpawnerLocations) (Variables|Default|GetSpawnerLocation))
      (Utilities|Array|Add (Variables|Default|GetCrawlingBoltHandLocations) (Variables|Default|GetBossHandLocation))
      (bind _returnvalue_2 (Audio|SpawnSoundAttached :Sound "/FNGameplayCues/Sounds/Apollo/Events/Dadbro/DadBro_CrawlingBolt_Summon_Cue.DadBro_CrawlingBolt_Summon_Cue" :AttachToComponent _returnvalue :AttachPointName "None" :LocationType "KeepRelativeOffset" :bStopWhenAttachedToDestroyed true :VolumeMultiplier 1.0 :PitchMultiplier 1.0 :StartTime 0.0 :bAutoDestroy false))
      (Utilities|Array|Add (Variables|Default|GetCrawling%20Bolt%20Audio) _returnvalue_2)
      (CallFunction|DrawHandandSpawnerLocationDebug) ; Draw debug if desired
      (bind (_direction _crawlingoffset) (Timeline|Timeline_0 PlayFromStart)
        (:Update
          (for (_array_element _array_index) (Variables|Default|GetCrawlingBoltArray)
            (Utilities|IsValid _array_element
              (:Is%20Valid
                (CallFunction|DrawFlightDebug (Utilities|Array|Get(aref) _crawlingbolthandlocations _array_index) (Utilities|Array|Get(acopy) _crawlingboltspawnerlocations _array_index) (Math|Vector|Lerp(Vector) (Utilities|Array|Get(aref) _crawlingbolthandlocations _array_index) (Utilities|Array|Get(acopy) _crawlingboltspawnerlocations _array_index) _crawlingoffset))
                (Transformation|SetWorldLocation _array_element (Math|Vector|Lerp(Vector) (Utilities|Array|Get(aref) _crawlingbolthandlocations _array_index) (Utilities|Array|Get(acopy) _crawlingboltspawnerlocations _array_index) _crawlingoffset) false false)))))
        (:Finished
          (for (_ _array_index_2) (Variables|Default|GetCrawlingBoltArray)
            (bind _returnvalue_3 (Effects|SpawnEmitteratLocation 0 "/FNGameplayCues/Athena/DADBRO/FX/Summon/P_DadBRO_Summon_CrawlingBoltActive.P_DadBRO_Summon_CrawlingBoltActive" (Utilities|Array|Get(acopy) (Variables|Default|GetCrawlingBoltSpawnerLocations) _array_index_2)))
            (Audio|SpawnSoundAttached "/Game/Sounds/Apollo/Events/Dadbro/DadBro_CrawlingBolt_Active_Cue.DadBro_CrawlingBolt_Active_Cue" _returnvalue_3))
          (Utilities|Array|Clear (Variables|Default|GetCrawlingBoltArray))
          (Utilities|Array|Clear (Variables|Default|GetCrawlingBoltSpawnerLocations))
          (for _array_element_2 (Variables|Default|GetCrawling%20Bolt%20Audio)
            (Utilities|IsValid _array_element_2
              (:Is%20Valid
                (Audio|Components|Audio|FadeOut _array_element_2 0.5 0.0 "Linear" :!class /Script/Engine.AudioComponent))))
          (Utilities|Array|Clear (Variables|Default|GetCrawling%20Bolt%20Audio)))))))

(event EventEndPlay
  (params (EndPlayReason :type EEndPlayReason%20Enum :default "Destroyed"))
  (for _array_element (Variables|Default|GetCrawling%20Bolt%20Audio)
    (Utilities|IsValid _array_element
      (:Is%20Valid
        (Audio|Components|Audio|FadeOut _array_element 0.5 0.0 "Linear" :!class /Script/Engine.AudioComponent))))
  (Utilities|Array|Clear (Variables|Default|GetCrawling%20Bolt%20Audio)))

(fn DrawFlightDebug
  (params (CurrentHandLocation :type Vector) (CurrentSpawnerLocation :type Vector) (NewTrailLocation :type Vector))
  (if (Variables|Default|GetbDebug)
    (Rendering|Debug|DrawDebugSphere 0 (Variables|GetCurrentSpawnerLocation) 4.5e+02 5 "(R=0.000000,G=0.081708,B=1.000000,A=1.000000)" 6e+01)
    (Rendering|Debug|DrawDebugSphere 0 (Variables|GetCurrentHandLocation) 1.5e+02 5 "(R=0.108358,G=1.000000,B=0.000000,A=1.000000)" 6e+01)
    (Rendering|Debug|DrawDebugSphere 0 (Variables|GetNewTrailLocation) 75.0 5 "(R=1.000000,G=0.986931,B=0.000000,A=1.000000)" 6e+01)))

(fn DrawHandAndSpawnerLocationDebug
  (if (Variables|Default|GetbDebug)
    (Rendering|Debug|DrawDebugSphere 0 (Variables|Default|GetSpawnerLocation) 4e+02 12 "(R=0.000000,G=0.081708,B=1.000000,A=1.000000)" 6e+01)
    (Rendering|Debug|DrawDebugSphere 0 (Variables|Default|GetBossHandLocation) 1e+02 12 "(R=0.108358,G=1.000000,B=0.000000,A=1.000000)" 6e+01)))

(fn ConstructionScript)
