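; Gameplay cue notify for the DADBRO sweep beam attack, drawing a beam with sound and ground impacts from the AI pawn's head.
; A once-guarded timeline line-traces each update along head pitch and roll with actor yaw, spawning or restretching the beam and moving its audio to the midpoint.
; CloseCheck spaces out hit particles, camera shakes and randomly rotated, fading decals by distance from the previous impact.
; Removal, timeline completion and destruction fade out the audio.
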(event EventDestroyed
  (bind _beam_audio (Variables|Default|GetBeam%20Audio)
    (:then
      (Audio|Components|Audio|FadeOut _beam_audio 0.75 0.0 "Linear" :!class /Script/Engine.AudioComponent))
    (:else)))

(event Custom|Reset
  (bind _trace_start (Variables|Default|GetTrace%20Start))
  (bind _impact_point (Variables|Default|GetImpact%20Point))
  (bind _returnvalue (- _impact_point _trace_start))
  (bind _returnvalue_2 (* _returnvalue 0.5))
  (bind _localpawn (Variables|Default|GetLocalPawn))
  (bind _mesh (Variables|/Script/Engine.Character|GetMesh _localpawn))
  (bind _returnvalue_3 (Transformation|GetSocketTransform _mesh "head" "RTS_World"))
  (bind _rotation (.rotation _returnvalue_3))
  (bind _returnvalue_4 (Transformation|GetActorRotation _localpawn))
  (bind _yaw (.yaw _returnvalue_4))
  (bind _pitch (.pitch _rotation))
  (bind _roll (.roll _rotation))
  (bind _returnvalue_5 (Math|Rotator|MakeRotator _roll _pitch _yaw))
  (bind _trace_start_2 (Variables|Default|GetTrace%20Start))
  (bind _returnvalue_6 (+ _trace_start_2 _returnvalue_2))
  (Utilities|FlowControl|Sequence
    (:then_0
      (Timeline|Overtime Stop
        (:Update
          (bind (_outhit _returnvalue_7) (Collision|LineTraceByChannel :WorldContextObject (Variables|Default|GetLocalPawn) :Start (.location _returnvalue_3) :End (+ (.location _returnvalue_3) (* (Math|Rotator|GetRotationXVector _returnvalue_5) 8096.0)) :TraceChannel "TraceTypeQuery1" :bTraceComplex false :DrawDebugType "None" :bIgnoreSelf true :TraceColor "(R=1.000000,G=0.000000,B=0.000000,A=1.000000)" :TraceHitColor "(R=0.000000,G=1.000000,B=0.000000,A=1.000000)" :DrawTime 1e+01))
          (if _returnvalue_7
            (Variables|Default|SetImpact%20Point (.ImpactPoint (Collision|BreakHitResult _outhit)))
            (Variables|Default|SetTrace%20Start (.TraceStart (Collision|BreakHitResult _outhit)))
            (bind _distance (Class|MissionBlueprintFunctionLibrary|DistanceBetweenTwoVectors (.TraceStart (Collision|BreakHitResult _outhit)) (.Location (Collision|BreakHitResult _outhit)) (Variables|Default|GetLocalPawn)))
            (bind _output_get (Variables|Default|SetStartTransform (Math|Transform|MakeTransform (.TraceStart (Collision|BreakHitResult _outhit)) _returnvalue_5 (Math|Vector|MakeVector (/ (Math|Float|Absolute(Float) _distance) 512.0) 1.0 1.0))))
            (Variables|Default|GetP_SweepBeam
              (:then
                (Transformation|SetRelativeTransform (Variables|Default|GetP_SweepBeam) (Variables|Default|GetStartTransform) false false) ; Update Beam Location
                (Transformation|SetRelativeTransform (Variables|Default|GetBeam%20Audio) (Math|Transform|MakeTransform _returnvalue_6 (Math|Rotator|CombineRotators (Math|Rotator|FindLookatRotation (Variables|Default|GetTrace%20Start) (Variables|Default|GetImpact%20Point)) "67.500000,67.500000,67.500000")) false false) ; Update Beam Location
              )
              (:else
                (bind _returnvalue_8 (Effects|SpawnEmitteratLocation (Variables|Default|GetLocalPawn) "/FNGameplayCues/Athena/DADBRO/FX/Sweep/P_DadBro_Sweep_Beam.P_DadBro_Sweep_Beam" (.location _output_get) (.rotation _output_get) (.scale _output_get))) ; Make new beam
                (Variables|Default|SetP_SweepBeam _returnvalue_8) ; Save Beam so we can update it.
                (bind _returnvalue_9 (Audio|SpawnSoundatLocation (Variables|Default|GetLocalPawn) (Variables|Default|GetBeam%20Sweep%20Sound) _returnvalue_6))
                (Variables|Default|SetBeam%20Audio _returnvalue_9)))
            (if (CallFunction|CloseCheck (.ImpactPoint (Collision|BreakHitResult _outhit)))
              (Effects|SpawnEmitteratLocation (Variables|Default|GetLocalPawn) "/Game/Athena/DADBRO/FX/Sweep/P_DadBro_Sweep_Hit.P_DadBro_Sweep_Hit" (+ (.ImpactPoint (Collision|BreakHitResult _outhit)) "0,0.000000,100.000000")) ; Impact FX on ground
              (Variables|Default|SetPrevious%20Impact (.ImpactPoint (Collision|BreakHitResult _outhit)))
              (Camera|PlayWorldCameraShake (Variables|Default|GetLocalPawn) "/FNGameplayCues/Effects/Fort_Effects/BluePrints/CameraShakes/B_Shudder_Vertical_Jolt_CameraShake.B_Shudder_Vertical_Jolt_CameraShake_C" (.ImpactPoint (Collision|BreakHitResult _outhit)) 5e+02 4e+03)
              (bind _returnvalue_10 (Rendering|Decal|SpawnDecalatLocation (Variables|Default|GetLocalPawn) "/Game/Athena/DADBRO/FX/Sweep/M_Dadbro_Char_Decal.M_Dadbro_Char_Decal" (Math|Vector|MakeVector (Variables|Default|GetDecalSize) (Variables|Default|GetDecalSize) (Variables|Default|GetDecalSize)) (.ImpactPoint (Collision|BreakHitResult _outhit)) (Math|Rotator|MakeRotator (Math|Float|Lerp 0.0 359.0 (Math|Random|RandomFloat)) -9e+01) 4.0))
              (Rendering|Components|Decal|SetFadeOut _returnvalue_10 2.0 2.0 true))))
        (:Finished
          (CallFunction|Reset)
          (bind _beam_audio (Variables|Default|GetBeam%20Audio)
            (:then
              (Audio|Components|Audio|FadeOut _beam_audio 0.75 0.0 "Linear" :!class /Script/Engine.AudioComponent))
            (:else)))) ; Shortened the duration of timeline from 1.75 to 1.4 to fix the impact FX roll back issue
    )
    (:then_1
      (Utilities|FlowControl|DoOnce false
        (:Completed
          (Timeline|Overtime PlayFromStart))))))

(event GameplayCueNotify|EventHandleGameplayCue
  (params (MyTarget :type Actor%20Object%20Reference) (EventType :type EGameplayCueEvent%20Enum :default "OnActive") (Parameters :type Gameplay%20Cue%20Parameters%20Structure%20%28by%20ref%29))
  (bind _trace_start (Variables|Default|GetTrace%20Start))
  (bind _impact_point (Variables|Default|GetImpact%20Point))
  (bind _returnvalue (- _impact_point _trace_start))
  (bind _returnvalue_2 (* _returnvalue 0.5))
  (bind _localpawn (Variables|Default|GetLocalPawn))
  (bind _mesh (Variables|/Script/Engine.Character|GetMesh _localpawn))
  (bind _returnvalue_3 (Transformation|GetSocketTransform _mesh "head" "RTS_World"))
  (bind _rotation (.rotation _returnvalue_3))
  (bind _returnvalue_4 (Transformation|GetActorRotation _localpawn))
  (bind _yaw (.yaw _returnvalue_4))
  (bind _pitch (.pitch _rotation))
  (bind _roll (.roll _rotation))
  (bind _returnvalue_5 (Math|Rotator|MakeRotator _roll _pitch _yaw))
  (bind _trace_start_2 (Variables|Default|GetTrace%20Start))
  (bind _returnvalue_6 (+ _trace_start_2 _returnvalue_2))
  (bind _asfort_aipawn (Utilities|Casting|CastToFortAIPawn MyTarget))
  (Variables|Default|SetLocalPawn _asfort_aipawn)
  (switch Utilities|FlowControl|Switch|SwitchonEGameplayCueEvent EventType
    (:OnActive
      _)
    (:Executed
      _)
    (:Removed
      (CallFunction|Reset)
      (bind _beam_audio (Variables|Default|GetBeam%20Audio)
        (:then
          (Audio|Components|Audio|FadeOut _beam_audio 0.75 0.0 "Linear" :!class /Script/Engine.AudioComponent))
        (:else))))
  (Utilities|FlowControl|DoOnce false
    (:Completed
      (Timeline|Overtime PlayFromStart))))

(fn CloseCheck
  (params (Impact%20Point :type Vector))
  (return (>= (Math|Vector|Distance(Vector) (Variables|Default|GetPrevious%20Impact) Impact%20Point) (LocalVariables|GetDecal%20Half%20Size))))

(fn ConstructionScript)
