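; Gameplay cue notify actor that shows the visual effect for the DADBRO fire area-of-effect, using an FX component and a decal.
; BeginPlay creates a dynamic material instance for the decal.
; On the Executed cue event it activates the AOE FX with reset and plays a timeline that drives the decal's BP_Offset parameter.
; The decal component is destroyed when the timeline finishes, and other cue events have no handling.
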(event EventBeginPlay
  (bind _returnvalue (Rendering|Components|Decal|CreateDynamicMaterialInstance (Variables|Default|GetDecal)))
  (Variables|Default|SetMID_Decal _returnvalue))

(event GameplayCueNotify|EventHandleGameplayCue
  (params (MyTarget :type Actor%20Object%20Reference) (EventType :type EGameplayCueEvent%20Enum :default "OnActive") (Parameters :type Gameplay%20Cue%20Parameters%20Structure%20%28by%20ref%29))
  (switch Utilities|FlowControl|Switch|SwitchonEGameplayCueEvent EventType
    (:Executed
      (Components|Activation|Activate (Variables|Default|GetFX_AOE) true)
      (bind (_direction _newtrack_0) (Timeline|Timeline_0 PlayFromStart)
        (:Update
          (Rendering|Material|SetScalarParameterValue (Variables|Default|GetMID_Decal) "BP_Offset" _newtrack_0 :!class /Script/Engine.MaterialInstanceDynamic))
        (:Finished
          (Components|DestroyComponent (Variables|Default|GetDecal) 0))))))

(fn ConstructionScript)
